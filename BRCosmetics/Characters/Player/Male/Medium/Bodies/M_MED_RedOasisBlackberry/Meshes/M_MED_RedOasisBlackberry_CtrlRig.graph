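# Control Rig for the RedOasisBlackberry medium male body cosmetic that procedurally drives limb twist bones.
# Upperarm and thigh twists counter-rotate using scaled swing-twist X extraction, with thighs measured against initial pose.
# Forearm and calf twist_01 bones take X-only rotation from weighted hand and calf constraints, passing half to twist_02.
# As wired, Begin Execution reaches only the left upperarm; the chain from right upperarm onward has no incoming execution link.
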ControlRigBlueprint /BRCosmetics/Characters/Player/Male/Medium/Bodies/M_MED_RedOasisBlackberry/Meshes/M_MED_RedOasisBlackberry_CtrlRig
graphs [
 RigVMFunctionLibrary {links [], nodes []}
 RigVMModel {nodes [
   RigUnit_BeginExecution = &RigUnit_BeginExecution()
   n1 = /Script/RigVM.RigVMFunction_MathQuaternionToAxisAndAngle(Value=&RigVMFunction_MathQuaternionSwingTwist(
     TwistAxis=(X=1,Y=0,Z=0)
     Input=&RigUnit_GetTransform(Item=(Type=Bone,Name="upperarm_l"), Space=LocalSpace, bInitial=false).Transform.Rotation
    ).Twist) {is=struct RigVMUnitNode, template="ToAxisAndAngle::Execute(in Value,out Axis,out Angle)"}
   RerouteNode_2_1 = RigVMDispatchNode(Value=-0.6) {template="DISPATCH_RigVMDispatch_Constant(io Value)"}
   Multiply = &RigVMFunction_MathDoubleMul(A=@n1.Angle, B=@RerouteNode_2_1.Value)
   Set Transform = &RigUnit_SetRotation(
    Item=(Type=Bone,Name="upperarm_twist_01_l")
    Weight=1
    bInitial=False
    Value=/Script/RigVM.RigVMFunction_MathQuaternionFromAxisAndAngle(Axis=@n1.Axis, Angle=@Multiply.Result) {is=struct RigVMUnitNode, template="FromAxisAndAngle::Execute(in Axis,in Angle,out Result)"}.Result
    ExecutePin=@RigUnit_BeginExecution.ExecutePin
   )
   n2 = &RigVMFunction_MathQuaternionToAxisAndAngle(Value=&RigVMFunction_MathQuaternionSwingTwist(
     TwistAxis=(X=1,Y=0,Z=0)
     Input=&RigUnit_GetTransform(Item=(Type=Bone,Name="upperarm_l"), Space=LocalSpace, bInitial=false).Transform.Rotation
    ).Twist) {template="ToAxisAndAngle::Execute(in Value,out Axis,out Angle)"}
   RerouteNode_2_1_1 = RigVMDispatchNode(Value=-0.5) {template="DISPATCH_RigVMDispatch_Constant(io Value)"}
   > Set Transform_1 = &RigUnit_SetRotation(
    Item=(Type=Bone,Name="upperarm_twist_02_l")
    Weight=1
    bInitial=False
    Value=&RigVMFunction_MathQuaternionFromAxisAndAngle(Axis=@n2.Axis, Angle=&RigVMFunction_MathDoubleMul(A=@n2.Angle, B=@RerouteNode_2_1_1.Value).Result) {template="FromAxisAndAngle::Execute(in Axis,in Angle,out Result)"}.Result
   )
   n3 = &RigVMFunction_MathQuaternionToAxisAndAngle(Value=&RigVMFunction_MathQuaternionSwingTwist(
     TwistAxis=(X=1,Y=0,Z=0)
     Input=&RigUnit_GetTransform(Item=(Type=Bone,Name="upperarm_r"), Space=LocalSpace, bInitial=false).Transform.Rotation
    ).Twist) {template="ToAxisAndAngle::Execute(in Value,out Axis,out Angle)"}
   Set Transform_2 = &RigUnit_SetRotation(
    ExecutePin=()
    Item=(Type=Bone,Name="upperarm_twist_01_r")
    Weight=1
    bInitial=False
    Value=&RigVMFunction_MathQuaternionFromAxisAndAngle(Axis=@n3.Axis, Angle=&RigVMFunction_MathDoubleMul(A=@n3.Angle, B=@RerouteNode_2_1.Value).Result) {template="FromAxisAndAngle::Execute(in Axis,in Angle,out Result)"}.Result
   )
   n4 = &RigVMFunction_MathQuaternionToAxisAndAngle(Value=&RigVMFunction_MathQuaternionSwingTwist(
     TwistAxis=(X=1,Y=0,Z=0)
     Input=&RigUnit_GetTransform(Item=(Type=Bone,Name="upperarm_r"), Space=LocalSpace, bInitial=false).Transform.Rotation
    ).Twist) {template="ToAxisAndAngle::Execute(in Value,out Axis,out Angle)"}
   > Set Transform_1_1 = &RigUnit_SetRotation(
    Item=(Type=Bone,Name="upperarm_twist_02_r")
    Weight=1
    bInitial=False
    Value=&RigVMFunction_MathQuaternionFromAxisAndAngle(Axis=@n4.Axis, Angle=&RigVMFunction_MathDoubleMul(A=@n4.Angle, B=@RerouteNode_2_1_1.Value).Result) {template="FromAxisAndAngle::Execute(in Axis,in Angle,out Result)"}.Result
   )
   n5 = &RigUnit_GetTransform(Item=(Type=Bone,Name="thigh_l"), Space=LocalSpace)
   n6 = &RigUnit_GetTransform(Item=(Type=Bone,Name="thigh_l"), Space=LocalSpace, bInitial=true)
   n7 = &RigVMFunction_MathQuaternionToAxisAndAngle(Value=&RigVMFunction_MathQuaternionSwingTwist(
     TwistAxis=(X=1,Y=0,Z=0)
     Input=&RigVMFunction_MathTransformMakeRelative(Global=@n5.Transform, Parent=@n6.Transform).Local.Rotation
    ).Twist) {template="ToAxisAndAngle::Execute(in Value,out Axis,out Angle)"}
   > Set Transform_1_1_1 = &RigUnit_SetRotation(
    Item=(Type=Bone,Name="thigh_twist_01_l")
    Weight=1
    bInitial=False
    Value=&RigVMFunction_MathQuaternionFromAxisAndAngle(Axis=@n7.Axis, Angle=&RigVMFunction_MathDoubleMul(B=-1, A=@n7.Angle).Result) {template="FromAxisAndAngle::Execute(in Axis,in Angle,out Result)"}.Result
   )
   n8 = &RigUnit_GetTransform(Item=(Type=Bone,Name="thigh_r"), Space=LocalSpace, bInitial=true)
   n9 = &RigUnit_GetTransform(Item=(Type=Bone,Name="thigh_r"), Space=LocalSpace)
   n10 = &RigVMFunction_MathQuaternionToAxisAndAngle(Value=&RigVMFunction_MathQuaternionSwingTwist(
     TwistAxis=(X=1,Y=0,Z=0)
     Input=&RigVMFunction_MathTransformMakeRelative(Parent=@n8.Transform, Global=@n9.Transform).Local.Rotation
    ).Twist) {template="ToAxisAndAngle::Execute(in Value,out Axis,out Angle)"}
   > Set Transform_1_1_1_1 = &RigUnit_SetRotation(
    Item=(Type=Bone,Name="thigh_twist_01_r")
    Weight=1
    bInitial=False
    Value=&RigVMFunction_MathQuaternionFromAxisAndAngle(Axis=@n10.Axis, Angle=&RigVMFunction_MathDoubleMul(B=-1, A=@n10.Angle).Result) {template="FromAxisAndAngle::Execute(in Axis,in Angle,out Result)"}.Result
   )
   > &RigUnit_ParentConstraint(
    AdvancedSettings=(RotationOrderForFilter=ZYX)
    Child=(Type=Bone,Name="lowerarm_twist_01_l")
    Filter=(TranslationFilter=(bX=false,bY=false,bZ=false),RotationFilter=(bX=True,bY=false,bZ=false),ScaleFilter=(bX=false,bY=false,bZ=false))
    Parents=((Item=(Type=Bone,Name="hand_l"),Weight=0.75),(Item=(Type=Bone,Name="lowerarm_twist_01_l"),Weight=0.25))
    Weight=1
   )
   n11 = &RigVMFunction_MathQuaternionToEuler(
    RotationOrder=ZYX
    Value=&RigUnit_GetTransform(Item=(Type=Bone,Name="lowerarm_twist_01_l"), Space=LocalSpace).Transform.Rotation
   )
   > &RigUnit_SetRotation(
    Item=(Type=Bone,Name="lowerarm_twist_01_l")
    Weight=1
    Value=&RigVMFunction_MathQuaternionFromEuler(Euler=(X=0,Y=0,Z=0), Euler.X=@n11.Result.X).Result
   )
   > &RigUnit_SetRotation(
    Item=(Type=Bone,Name="lowerarm_twist_02_l")
    Weight=1
    Value=&RigVMFunction_MathQuaternionFromEuler(
     Euler=(X=0,Y=0,Z=0)
     Euler.X=&RigVMFunction_MathFloatMul(B=0.5, A=@n11.Result.X) {template="Multiply::Execute(in A,in B,out Result)"}.Result
    ).Result
   )
   > &RigUnit_ParentConstraint(
    AdvancedSettings=(RotationOrderForFilter=ZYX)
    Child=(Type=Bone,Name="lowerarm_twist_01_r")
    Filter=(TranslationFilter=(bX=false,bY=false,bZ=false),RotationFilter=(bX=True,bY=false,bZ=false),ScaleFilter=(bX=false,bY=false,bZ=false))
    Parents=((Item=(Type=Bone,Name="hand_r"),Weight=0.75),(Item=(Type=Bone,Name="lowerarm_twist_01_r"),Weight=0.25))
    Weight=1
   )
   n12 = &RigVMFunction_MathQuaternionToEuler(
    RotationOrder=ZYX
    Value=&RigUnit_GetTransform(Item=(Type=Bone,Name="lowerarm_twist_01_r"), Space=LocalSpace).Transform.Rotation
   )
   > &RigUnit_SetRotation(
    Item=(Type=Bone,Name="lowerarm_twist_01_r")
    Weight=1
    Value=&RigVMFunction_MathQuaternionFromEuler(Euler=(X=0,Y=0,Z=0), Euler.X=@n12.Result.X).Result
   )
   > &RigUnit_SetRotation(
    Item=(Type=Bone,Name="lowerarm_twist_02_r")
    Weight=1
    Value=&RigVMFunction_MathQuaternionFromEuler(
     Euler=(X=0,Y=0,Z=0)
     Euler.X=&RigVMFunction_MathFloatMul(B=0.5, A=@n12.Result.X) {template="Multiply::Execute(in A,in B,out Result)"}.Result
    ).Result
   )
   > &RigUnit_ParentConstraint(
    AdvancedSettings=(RotationOrderForFilter=ZYX)
    Child=(Type=Bone,Name="calf_twist_01_l")
    Filter=(TranslationFilter=(bX=false,bY=false,bZ=false),RotationFilter=(bX=True,bY=false,bZ=false),ScaleFilter=(bX=false,bY=false,bZ=false))
    Parents=((Item=(Type=Bone,Name="calf_l"),Weight=0.75),(Item=(Type=Bone,Name="calf_twist_01_l"),Weight=0.25))
    Weight=1
   )
   n13 = &RigVMFunction_MathQuaternionToEuler(
    RotationOrder=ZYX
    Value=&RigUnit_GetTransform(Item=(Type=Bone,Name="calf_twist_01_l"), Space=LocalSpace).Transform.Rotation
   )
   > &RigUnit_SetRotation(
    Item=(Type=Bone,Name="calf_twist_01_l")
    Weight=1
    Value=&RigVMFunction_MathQuaternionFromEuler(Euler=(X=0,Y=0,Z=0), Euler.X=@n13.Result.X).Result
   )
   > &RigUnit_SetRotation(
    Item=(Type=Bone,Name="calf_twist_02_l")
    Weight=1
    Value=&RigVMFunction_MathQuaternionFromEuler(
     Euler=(X=0,Y=0,Z=0)
     Euler.X=&RigVMFunction_MathFloatMul(B=0.5, A=@n13.Result.X) {template="Multiply::Execute(in A,in B,out Result)"}.Result
    ).Result
   )
   > &RigUnit_ParentConstraint(
    AdvancedSettings=(RotationOrderForFilter=ZYX)
    Child=(Type=Bone,Name="calf_twist_01_r")
    Filter=(TranslationFilter=(bX=false,bY=false,bZ=false),RotationFilter=(bX=True,bY=false,bZ=false),ScaleFilter=(bX=false,bY=false,bZ=false))
    Parents=((Item=(Type=Bone,Name="calf_r"),Weight=0.75),(Item=(Type=Bone,Name="calf_twist_01_r"),Weight=0.25))
    Weight=1
   )
   n14 = &RigVMFunction_MathQuaternionToEuler(
    RotationOrder=ZYX
    Value=&RigUnit_GetTransform(Item=(Type=Bone,Name="calf_twist_01_r"), Space=LocalSpace).Transform.Rotation
   )
   > &RigUnit_SetRotation(
    Item=(Type=Bone,Name="calf_twist_01_r")
    Weight=1
    Value=&RigVMFunction_MathQuaternionFromEuler(Euler=(X=0,Y=0,Z=0), Euler.X=@n14.Result.X).Result
   )
   > &RigUnit_SetRotation(
    Item=(Type=Bone,Name="calf_twist_02_r")
    Weight=1
    Value=&RigVMFunction_MathQuaternionFromEuler(
     Euler=(X=0,Y=0,Z=0)
     Euler.X=&RigVMFunction_MathFloatMul(B=0.5, A=@n14.Result.X) {template="Multiply::Execute(in A,in B,out Result)"}.Result
    ).Result
   )
  ]}
]
hierarchy [
 root {loc=<0 0 0>, rot=<0 0 0 1>}
 pelvis {parent=root, loc=<0 -0.7071 93.1191>, rot=<0.0127 0.707 -0.0127 -0.707>}
 spine_01 {parent=pelvis, loc=<6.7494 -0.6388 0>, rot=<0 0 0.034 -0.9994>}
 spine_02 {parent=spine_01, loc=<7.2518 -0.035 0>, rot=<0 0 0.0238 -0.9997>}
 spine_03 {parent=spine_02, loc=<6.0104 0 0>, rot=<0 0 0.0149 -0.9999>}
 spine_04 {parent=spine_03, loc=<8.1024 0 0>, rot=<0 0 -0.0132 -0.9999>}
 spine_05 {parent=spine_04, loc=<10.0146 0 0>, rot=<0 0 -0.0015 -1>}
 attach_backpack {parent=spine_05, loc=<-0.123 -13.473 0>, rot=<-0.0093 -0.707 0.0093 -0.707>}
 neck_01 {parent=spine_05, loc=<11.3138 0.1529 0>, rot=<0 0 -0.2003 -0.9797>}
 neck_02 {parent=neck_01, loc=<4.5627 0 0>, rot=<0 0 0.0504 -0.9987>}
 head {parent=neck_02, loc=<5.3914 0 0>, rot=<0 0 0.115 -0.9934>}
 clavicle_l {parent=spine_05, loc=<4.3575 -0.1466 -2.3223>, rot=<0.0465 -0.7289 -0.009 -0.6829>}
 upperarm_l {parent=clavicle_l, loc=<15.5111 0 0>, rot=<0.0226 -0.3234 -0.0088 -0.9459>}
 lowerarm_l {parent=upperarm_l, loc=<26.6514 0 0>, rot=<0.0391 -0.0099 -0.2453 -0.9686>}
 hand_l {parent=lowerarm_l, loc=<22.3491 0 0>, rot=<-0.7081 -0.0507 0.0975 -0.6975>}
 thumb_01_l {parent=hand_l, loc=<1.3118 -0.3191 -2.459>, rot=<0.6546 -0.4042 0.0664 -0.6354>}
 thumb_02_l {parent=thumb_01_l, loc=<5.9336 -0.0448 0.0401>, rot=<-0.0037 0.0213 0.298 -0.9543>}
 thumb_03_l {parent=thumb_02_l, loc=<2.2223 0.3682 -0.0907>, rot=<0.0118 -0.013 0.0066 -0.9998>}
 index_metacarpal_l {parent=hand_l, loc=<1.6535 0.0183 -2.96>, rot=<0.0158 -0.0124 -0.0173 -0.9996>}
 index_01_l {parent=index_metacarpal_l, loc=<8.3008 0.0357 -0.0524>, rot=<-0.0044 0.0485 0.0924 -0.9945>}
 index_02_l {parent=index_01_l, loc=<3.3587 -0.3004 0.0452>, rot=<-0.002 0.0027 0.3225 -0.9466>}
 index_03_l {parent=index_02_l, loc=<2.5578 -0.123 -0.0252>, rot=<-0.002 0.0067 0.2558 -0.9667>}
 middle_metacarpal_l {parent=hand_l, loc=<2.1856 -0.1725 -0.5984>, rot=<0.006 0.0168 -0.0329 -0.9993>}
 middle_01_l {parent=middle_metacarpal_l, loc=<7.3275 -0.2447 0.0129>, rot=<-0.0208 0.0042 0.0858 -0.9961>}
 middle_02_l {parent=middle_01_l, loc=<4.0123 0.1401 0.1175>, rot=<-0.0218 0.0258 0.3512 -0.9357>}
 middle_03_l {parent=middle_02_l, loc=<2.6077 0.0534 0.1373>, rot=<0.0082 0.0188 0.1708 -0.9851>}
 ring_metacarpal_l {parent=hand_l, loc=<2.2074 -0.3625 1.3638>, rot=<-0.0284 0.0722 -0.0458 -0.9959>}
 ring_01_l {parent=ring_metacarpal_l, loc=<7.0295 -0.6197 -0.1446>, rot=<-0.0489 -0.0252 0.1892 -0.9804>}
 ring_02_l {parent=ring_01_l, loc=<3.6174 0.4647 -0.1027>, rot=<0 0 0.2806 -0.9598>}
 ring_03_l {parent=ring_02_l, loc=<2.4152 0.114 0.0199>, rot=<0 0 0.1189 -0.9929>}
 pinky_metacarpal_l {parent=hand_l, loc=<1.7616 -0.9753 3.3814>, rot=<-0.0397 0.0814 -0.069 -0.9935>}
 pinky_01_l {parent=pinky_metacarpal_l, loc=<7.1709 -0.4831 -0.1094>, rot=<-0.0654 -0.0035 0.2242 -0.9723>}
 pinky_02_l {parent=pinky_01_l, loc=<2.8274 0.0893 -0.1043>, rot=<0.0037 0.0011 0.2854 -0.9584>}
 pinky_03_l {parent=pinky_02_l, loc=<2.2046 0.0307 -0.0078>, rot=<-0.0002 0.0098 0.1445 -0.9895>}
 lowerarm_twist_02_l {parent=lowerarm_l, loc=<9.6199 0 0>, rot=<0 0 0 -1>}
 lowerarm_twist_01_l {parent=lowerarm_l, loc=<17.8549 0 0>, rot=<0 0 0 -1>}
 deform_elbow_in_base_l {parent=lowerarm_l, loc=<0.2112 5.259 0.8274>, rot=<0 0 0 1>}
 deform_elbow_in_l {parent=deform_elbow_in_base_l, loc=<0 0 0>, rot=<0 0 0 1>}
 deform_elbow_out_base_l {parent=lowerarm_l, loc=<0 -4 0>, rot=<0 0 0 1>}
 deform_elbow_out_l {parent=deform_elbow_out_base_l, loc=<0 0 0>, rot=<0 0 0 1>}
 upperarm_twist_01_l {parent=upperarm_l, loc=<0.0041 0 0>, rot=<0 0 0 1>}
 deform_upperarm_bk_base_l {parent=upperarm_twist_01_l, loc=<1.0813 -7.2193 -1.9453>, rot=<0 0 0 1>}
 deform_upperarm_bk_l {parent=deform_upperarm_bk_base_l, loc=<0 0 0>, rot=<0 0 0 1>}
 deform_upperarm_fr_base_l {parent=upperarm_twist_01_l, loc=<-0.3691 7.0506 -2.765>, rot=<0 0 0 1>}
 deform_upperarm_fr_l {parent=deform_upperarm_fr_base_l, loc=<0 0 0>, rot=<0 0 0 1>}
 deform_upperarm_up_base_l {parent=upperarm_twist_01_l, loc=<-3.2216 -1.1391 3.3311>, rot=<0 0 0 1>}
 deform_upperarm_up_l {parent=deform_upperarm_up_base_l, loc=<0 0 0>, rot=<0 0 0 1>}
 upperarm_twist_02_l {parent=upperarm_l, loc=<13.7941 0 0>, rot=<0 0 0 1>}
 attach_cape {parent=spine_05, loc=<9.017 -11.714 0>, rot=<0.0616 -0.7044 -0.0616 -0.7044>}
 clavicle_r {parent=spine_05, loc=<4.3575 -0.1466 2.3223>, rot=<0.7289 0.0465 -0.6829 0.009>}
 upperarm_r {parent=clavicle_r, loc=<-15.5111 0 0>, rot=<0.0226 -0.3234 -0.0088 -0.9459>}
 lowerarm_r {parent=upperarm_r, loc=<-26.6514 0 0>, rot=<0.0391 -0.0099 -0.2453 -0.9686>}
 hand_r {parent=lowerarm_r, loc=<-22.3491 0 0>, rot=<-0.7081 -0.0507 0.0975 -0.6975>}
 thumb_01_r {parent=hand_r, loc=<-1.3118 0.3191 2.459>, rot=<0.6546 -0.4042 0.0664 -0.6354>}
 thumb_02_r {parent=thumb_01_r, loc=<-5.9336 0.0448 -0.0401>, rot=<-0.0037 0.0213 0.298 -0.9543>}
 thumb_03_r {parent=thumb_02_r, loc=<-2.2223 -0.3682 0.0907>, rot=<0.0118 -0.013 0.0066 -0.9998>}
 index_metacarpal_r {parent=hand_r, loc=<-1.6535 -0.0183 2.96>, rot=<0.0158 -0.0124 -0.0173 -0.9996>}
 index_01_r {parent=index_metacarpal_r, loc=<-8.3008 -0.0357 0.0524>, rot=<-0.0044 0.0485 0.0924 -0.9945>}
 index_02_r {parent=index_01_r, loc=<-3.3587 0.3004 -0.0452>, rot=<-0.002 0.0027 0.3225 -0.9466>}
 index_03_r {parent=index_02_r, loc=<-2.5578 0.123 0.0252>, rot=<-0.002 0.0067 0.2558 -0.9667>}
 middle_metacarpal_r {parent=hand_r, loc=<-2.1856 0.1725 0.5984>, rot=<0.006 0.0168 -0.0329 -0.9993>}
 middle_01_r {parent=middle_metacarpal_r, loc=<-7.3275 0.2447 -0.0129>, rot=<-0.0208 0.0042 0.0858 -0.9961>}
 middle_02_r {parent=middle_01_r, loc=<-4.0123 -0.1401 -0.1175>, rot=<-0.0218 0.0258 0.3512 -0.9357>}
 middle_03_r {parent=middle_02_r, loc=<-2.6077 -0.0534 -0.1373>, rot=<0.0082 0.0188 0.1708 -0.9851>}
 ring_metacarpal_r {parent=hand_r, loc=<-2.2074 0.3625 -1.3638>, rot=<-0.0284 0.0722 -0.0458 -0.9959>}
 ring_01_r {parent=ring_metacarpal_r, loc=<-7.0295 0.6197 0.1446>, rot=<-0.0489 -0.0252 0.1892 -0.9804>}
 ring_02_r {parent=ring_01_r, loc=<-3.6174 -0.4647 0.1027>, rot=<0 0 0.2806 -0.9598>}
 ring_03_r {parent=ring_02_r, loc=<-2.4152 -0.114 -0.0199>, rot=<0 0 0.1189 -0.9929>}
 pinky_metacarpal_r {parent=hand_r, loc=<-1.7616 0.9753 -3.3814>, rot=<-0.0397 0.0814 -0.069 -0.9935>}
 pinky_01_r {parent=pinky_metacarpal_r, loc=<-7.1709 0.4831 0.1094>, rot=<-0.0654 -0.0035 0.2242 -0.9723>}
 pinky_02_r {parent=pinky_01_r, loc=<-2.8274 -0.0893 0.1043>, rot=<0.0037 0.0011 0.2854 -0.9584>}
 pinky_03_r {parent=pinky_02_r, loc=<-2.2046 -0.0307 0.0078>, rot=<-0.0002 0.0098 0.1445 -0.9895>}
 lowerarm_twist_02_r {parent=lowerarm_r, loc=<-9.6199 0 0>, rot=<0 0 0 -1>}
 lowerarm_twist_01_r {parent=lowerarm_r, loc=<-17.8549 0 0>, rot=<0 0 0 -1>}
 deform_elbow_in_base_r {parent=lowerarm_r, loc=<-0.2111 -5.2591 -0.8275>, rot=<0 0 0 -1>}
 deform_elbow_in_r {parent=deform_elbow_in_base_r, loc=<-0.0001 0 0.0002>, rot=<0 0 0 1>}
 deform_elbow_out_base_r {parent=lowerarm_r, loc=<0 4 -0.0001>, rot=<0 0 0 -1>}
 deform_elbow_out_r {parent=deform_elbow_out_base_r, loc=<0 0 0>, rot=<0 0 0 1>}
 upperarm_twist_01_r {parent=upperarm_r, loc=<-0.0041 0 0>, rot=<0 0 0 -1>}
 deform_upperarm_bk_base_r {parent=upperarm_twist_01_r, loc=<-1.0814 7.2192 1.9456>, rot=<0 0 0 -1>}
 deform_upperarm_bk_r {parent=deform_upperarm_bk_base_r, loc=<0 0 0>, rot=<0 0 0 1>}
 deform_upperarm_fr_base_r {parent=upperarm_twist_01_r, loc=<0.3689 -7.0506 2.7652>, rot=<0 0 0 -1>}
 deform_upperarm_fr_r {parent=deform_upperarm_fr_base_r, loc=<0 0 0>, rot=<0 0 0 1>}
 deform_upperarm_up_base_r {parent=upperarm_twist_01_r, loc=<3.2214 1.1391 -3.3308>, rot=<0 0 0 1>}
 deform_upperarm_up_r {parent=deform_upperarm_up_base_r, loc=<0 0 0>, rot=<0 0 0 -1>}
 upperarm_twist_02_r {parent=upperarm_r, loc=<-13.7941 0 0>, rot=<0 0 0 -1>}
 dyn_zipper_MSTR {parent=spine_05, loc=<6.384 8.235 0.0005>, rot=<0 0 0 1>}
 dyn_zipper_1 {parent=dyn_zipper_MSTR, loc=<0 0 0>, rot=<0 0 0 1>}
 thigh_l {parent=pelvis, loc=<-2.5693 -0.0109 -7.7501>, rot=<0.0012 0.0572 0.0155 -0.9982>}
 calf_l {parent=thigh_l, loc=<-39.0484 0 0>, rot=<0.0006 -0.0161 -0.0391 -0.9991>}
 foot_l {parent=calf_l, loc=<-42.561 0.1077 0.4292>, rot=<0.0418 -0.0455 0.0343 -0.9975>}
 ball_l {parent=foot_l, loc=<-6.1599 13.8264 0.6806>, rot=<0.0001 -0.0001 -0.7063 -0.7079>}
 calf_twist_02_l {parent=calf_l, loc=<-18.8064 -0.0967 0.0397>, rot=<0.0237 0.0095 -0.0035 -0.9997>}
 calf_twist_01_l {parent=calf_l, loc=<-35.813 -0.0967 0.0397>, rot=<0.0237 0.0095 -0.0035 -0.9997>}
 deform_knee_in_base_l {parent=calf_l, loc=<1.298 -5.2957 -0.0496>, rot=<0 0 0 1>}
 deform_knee_in_l {parent=deform_knee_in_base_l, loc=<0 0 0>, rot=<0 0 0 1>}
 deform_knee_out_base_l {parent=calf_l, loc=<0 4 0>, rot=<0 0 0 1>}
 deform_knee_out_l {parent=deform_knee_out_base_l, loc=<0 0 0>, rot=<0 0 0 1>}
 thigh_twist_01_l {parent=thigh_l, loc=<-21.6303 0 0>, rot=<0.0013 0 0 -1>}
 dyn_shorts_hem_l_MSTR {parent=thigh_l, loc=<-20.3615 0 0>, rot=<0 0 0 1>}
 dyn_shorts_hem_l {parent=dyn_shorts_hem_l_MSTR, loc=<0 0 0>, rot=<0 0 0 1>}
 thigh_r {parent=pelvis, loc=<-2.5693 -0.0109 7.7501>, rot=<0.0572 -0.0012 0.9982 0.0155>}
 calf_r {parent=thigh_r, loc=<39.0484 0 0>, rot=<0.0006 -0.0161 -0.0391 -0.9991>}
 foot_r {parent=calf_r, loc=<42.561 -0.1077 -0.4292>, rot=<0.0418 -0.0455 0.0343 -0.9975>}
 ball_r {parent=foot_r, loc=<6.1599 -13.8264 -0.6806>, rot=<0.0001 -0.0001 -0.7063 -0.7079>}
 calf_twist_02_r {parent=calf_r, loc=<18.8064 0.0967 -0.0397>, rot=<0.0237 0.0095 -0.0035 -0.9997>}
 calf_twist_01_r {parent=calf_r, loc=<35.813 0.0967 -0.0397>, rot=<0.0237 0.0095 -0.0035 -0.9997>}
 deform_knee_in_base_r {parent=calf_r, loc=<-1.298 5.2957 0.0496>, rot=<0 0 0 -1>}
 deform_knee_in_r {parent=deform_knee_in_base_r, loc=<0 0 0>, rot=<0 0 0 1>}
 deform_knee_out_base_r {parent=calf_r, loc=<0 -4 0>, rot=<0 0 0 -1>}
 deform_knee_out_r {parent=deform_knee_out_base_r, loc=<0 0 0>, rot=<0 0 0 1>}
 thigh_twist_01_r {parent=thigh_r, loc=<21.6303 0 0>, rot=<0.0013 0 0 -1>}
 dyn_shorts_hem_r_MSTR {parent=thigh_r, loc=<20.358 0 0>, rot=<0 0 0 1>}
 dyn_shorts_hem_r {parent=dyn_shorts_hem_r_MSTR, loc=<0 0 0>, rot=<0 0 0 1>}
 deform_glute_base_l {parent=pelvis, loc=<-8.8064 -8.8222 -8.4287>, rot=<0.0275 0.0501 -0.4472 -0.8926>}
 deform_glute_l {parent=deform_glute_base_l, loc=<0 0 0>, rot=<0 0 0 1>}
 deform_glute_base_r {parent=pelvis, loc=<-8.8062 -8.8222 8.4287>, rot=<0.0501 -0.0275 0.8926 -0.4472>}
 deform_glute_r {parent=deform_glute_base_r, loc=<0 0 0>, rot=<0 0 0 1>}
 dyn_hem_fr_l_MSTR {parent=pelvis, loc=<0.9607 8.1161 -8.2537>, rot=<0.6244 0.4098 -0.5425 -0.3846>}
 dyn_hem_fr_l {parent=dyn_hem_fr_l_MSTR, loc=<0 0 0>, rot=<0 0 0 1>}
 dyn_hem_fr_MSTR {parent=pelvis, loc=<0.4194 9.416 0>, rot=<-0.5089 -0.491 -0.491 -0.5089>}
 dyn_hem_fr {parent=dyn_hem_fr_MSTR, loc=<0 0 0>, rot=<0 0 0 1>}
 dyn_hem_fr_r_MSTR {parent=pelvis, loc=<0.9607 8.1161 8.2537>, rot=<0.6244 0.4098 0.5425 0.3846>}
 dyn_hem_fr_r {parent=dyn_hem_fr_r_MSTR, loc=<0 0 0>, rot=<0 0 0 1>}
 dyn_hem_r_MSTR {parent=pelvis, loc=<2.0818 0.441 14.4661>, rot=<0.707 -0.0127 0.707 0.0127>}
 dyn_hem_r {parent=dyn_hem_r_MSTR, loc=<0 0 0>, rot=<0 0 0 1>}
 dyn_hem_l_MSTR {parent=pelvis, loc=<2.0818 0.441 -14.4661>, rot=<0.707 -0.0127 -0.707 -0.0127>}
 dyn_hem_l {parent=dyn_hem_l_MSTR, loc=<0 0 0>, rot=<0 0 0 1>}
]
variables [UpperarmTwist {type=double, default=0}]
