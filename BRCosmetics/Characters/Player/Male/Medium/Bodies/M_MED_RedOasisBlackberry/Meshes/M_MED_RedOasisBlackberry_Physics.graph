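# Physics asset for the male medium RedOasisBlackberry body cosmetic, covering its dynamic hem, shorts and zipper bones.
# Capsules on pelvis, spine_05 and thighs form the core body, with small capsules on five hem bones and a sphere on the zipper.
# Hems constrain to the pelvis and the zipper to spine_05 with angular limits.
# Each shorts hem uses four offset spheres in a square, constrained to its thigh with free rotation.
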PhysicsAsset /BRCosmetics/Characters/Player/Male/Medium/Bodies/M_MED_RedOasisBlackberry/Meshes/M_MED_RedOasisBlackberry_Physics
bodies {
 dyn_hem_fr [{shapeType=Capsule, length=3.25, radius=4, shapeName=dyn_hem_fr_capsule, rotation=<rot 90 153.4349 -116.5651>}]
 dyn_hem_fr_l [{shapeType=Capsule, length=3.25, radius=4, shapeName=dyn_hem_fr_l_capsule, rotation=<rot 90 180 -90>}]
 dyn_hem_fr_r [{shapeType=Capsule, length=3.25, radius=4, shapeName=dyn_hem_fr_r_capsule, rotation=<rot 90 180 -90>}]
 dyn_hem_l [{shapeType=Capsule, length=3.25, radius=4, shapeName=dyn_hem_l_capsule, rotation=<rot 90 153.4349 -116.5651>}]
 dyn_hem_r [{shapeType=Capsule, length=3.25, radius=4, shapeName=dyn_hem_r_capsule, rotation=<rot 90 153.4349 -116.5651>}]
 dyn_shorts_hem_l [
  {shapeType=Sphere, radius=4, shapeName=dyn_shorts_hem_l_sphere, center=<0 8 8>}
  {shapeType=Sphere, radius=4, shapeName=dyn_shorts_hem_l_sphere, center=<0 8 -8>}
  {shapeType=Sphere, radius=4, shapeName=dyn_shorts_hem_l_sphere, center=<0 -8 8>}
  {shapeType=Sphere, radius=4, shapeName=dyn_shorts_hem_l_sphere, center=<0 -8 -8>}
 ]
 dyn_shorts_hem_r [
  {shapeType=Sphere, radius=4, shapeName=dyn_shorts_hem_r_sphere, center=<0 8 8>}
  {shapeType=Sphere, radius=4, shapeName=dyn_shorts_hem_r_sphere, center=<0 8 -8>}
  {shapeType=Sphere, radius=4, shapeName=dyn_shorts_hem_r_sphere, center=<0 -8 8>}
  {shapeType=Sphere, radius=4, shapeName=dyn_shorts_hem_r_sphere, center=<0 -8 -8>}
 ]
 dyn_zipper_1 [{shapeType=Sphere, radius=1.255, shapeName=dyn_zipper_1_sphere, center=<-0.8638 1.2088 0>}]
 pelvis [{shapeType=Capsule, length=13.6814, radius=9.3641, shapeName=pelvis_capsule, center=<-0.3399 -1.2126 0>}]
 spine_05 [{shapeType=Capsule, length=16.6976, radius=7.6018, shapeName=spine_05_capsule, center=<-0.7062 3.044 0>}]
 thigh_l [{shapeType=Capsule, length=30, radius=8.2, center=<-15 -0.3971 0.0092>, rotation=<rot 90 180 0>}]
 thigh_r [{shapeType=Capsule, length=30, radius=8, center=<15 0 0>, rotation=<rot 90 0 0>}]
}
constraints [
 {
  bone1Name=thigh_l
  bone2Name=pelvis
  swing1LimitDegrees=45
  swing1Motion=Free
  swing2LimitDegrees=45
  swing2Motion=Free
  twistLimitDegrees=45
  twistMotion=Free
 }
 {
  bone1Name=thigh_r
  bone2Name=pelvis
  swing1LimitDegrees=45
  swing1Motion=Free
  swing2LimitDegrees=45
  swing2Motion=Free
  twistLimitDegrees=45
  twistMotion=Free
 }
 {
  bone1Name=dyn_shorts_hem_l
  bone2Name=thigh_l
  swing1LimitDegrees=45
  swing1Motion=Free
  swing2LimitDegrees=45
  swing2Motion=Free
  twistLimitDegrees=45
  twistMotion=Free
 }
 {
  bone1Name=dyn_shorts_hem_r
  bone2Name=thigh_r
  swing1LimitDegrees=45
  swing1Motion=Free
  swing2LimitDegrees=45
  swing2Motion=Free
  twistLimitDegrees=45
  twistMotion=Free
 }
 {bone1Name=dyn_zipper_1, bone2Name=spine_05, swing1LimitDegrees=45, swing2LimitDegrees=10, twistLimitDegrees=5}
 {
  bone1Name=spine_05
  bone2Name=pelvis
  swing1LimitDegrees=45
  swing1Motion=Free
  swing2LimitDegrees=45
  swing2Motion=Free
  twistLimitDegrees=45
  twistMotion=Free
 }
 {bone1Name=dyn_hem_fr, bone2Name=pelvis, swing1LimitDegrees=70, swing2LimitDegrees=70, twistLimitDegrees=5}
 {bone1Name=dyn_hem_fr_l, bone2Name=pelvis, swing1LimitDegrees=70, swing2LimitDegrees=70, twistLimitDegrees=5}
 {bone1Name=dyn_hem_fr_r, bone2Name=pelvis, swing1LimitDegrees=70, swing2LimitDegrees=70, twistLimitDegrees=5}
 {bone1Name=dyn_hem_l, bone2Name=pelvis, swing1LimitDegrees=70, swing2LimitDegrees=70, twistLimitDegrees=5}
 {bone1Name=dyn_hem_r, bone2Name=pelvis, swing1LimitDegrees=70, swing2LimitDegrees=70, twistLimitDegrees=5}
]
