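# Animation Blueprint for the RedOasisBlackberry male medium body cosmetic, adding secondary motion and clothing correctives to the parent mesh's copied pose.
# Two Control Rigs feed a RigidBody simulation that switches off during emotes.
# Bone offsets then adjust shorts and hems for movement, crouching and thigh lift, measured each update from ref-pose rotation deltas.
# Further tweaks shift backpack and cape attachments, constrain the zipper, and offset the back neck, more during melee.
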AnimBlueprint /BRCosmetics/Characters/Player/Male/Medium/Bodies/M_MED_RedOasisBlackberry/Meshes/M_MED_RedOasisBlackberry_AnimBP
target_skeleton=&/M_MED_RedOasisBlackberry_Skeleton
graphs [
 AnimGraph {nodes [
   n1 = Control Rig(Source=Copy Pose From Mesh() {bUseAttachedParent=true}) {
    alphaBoolBlend {blendOption=Linear}
    controlRigAssetReference {blueprintRigClass=&/M_MED_RedOasisBlackberry_CtrlRig.M_MED_RedOasisBlackberry_CtrlRig_C}
   }
   n2 = Control Rig(Source=@n1) {
    alphaBoolBlend {blendOption=Linear}
    controlRigAssetReference {blueprintRigClass=&/M_MED_RedOasisBlackberry_deform_ControlRig.M_MED_RedOasisBlackberry_deform_ControlRig_C}
   }
   n3 = RigidBody!(
    ComponentPose=Local To Component(LocalPose=@n2)
    bAlphaBoolEnabled:bool=Get EmoteCompensationAlpha() -> (EmoteCompensationAlpha:bool)
    ExternalForce
   ) -> (Pose) {
    alphaInputType=Bool
    bOverrideWorldGravity=true
    overridePhysicsAsset=&/M_MED_RedOasisBlackberry_Physics
    alphaBoolBlend {blendOption=Linear}
    componentAppliedLinearAccClamp=<3000 3000 3000>
    componentLinearAccScale=<0.3 0.3 0.3>
    componentLinearVelScale=<0.3 0.3 0.3>
    overrideWorldGravity {z=-980}
   }
   n4 = Get bIsMoving2D()
   n5 = "Transform (Modify) Bone - Bone: dyn_shorts_hem_r"!(ComponentPose=@n3, bAlphaBoolEnabled:bool=@n4, Translation="0,0,0", Rotation="0,0,0", Scale="1,1.1,1.1") -> (Pose) {
    alphaInputType=Bool
    rotationMode=BMM_Additive
    scaleMode=BMM_Additive
    translationMode=BMM_Additive
    alphaBoolBlend {blendInTime=0.3, blendOption=ExpInOut, blendOutTime=0.3}
    alphaScaleBiasClamp {bInterpResult=true}
    boneToModify {boneName=dyn_shorts_hem_r}
    scale {y=1.1, z=1.1}
   }
   n6 = "Transform (Modify) Bone - Bone: dyn_shorts_hem_l"!(ComponentPose=@n5, bAlphaBoolEnabled:bool=@n4, Translation="0,0,0", Rotation="0,0,0", Scale="1,1.1,1.1") -> (Pose) {
    alphaInputType=Bool
    rotationMode=BMM_Additive
    scaleMode=BMM_Additive
    translationMode=BMM_Additive
    alphaBoolBlend {blendInTime=0.3, blendOption=ExpInOut, blendOutTime=0.3}
    alphaScaleBiasClamp {bInterpResult=true}
    boneToModify {boneName=dyn_shorts_hem_l}
    scale {y=1.1, z=1.1}
   }
   n7 = Get IsCrouchingNotMoving() -> (IsCrouchingNotMoving:bool)
   n8 = "Transform (Modify) Bone - Bone: dyn_shorts_hem_r"(ComponentPose=@n6, bAlphaBoolEnabled=@n7, Translation="0,-3,0", Rotation="0,0,0", Scale="1,1,1") {
    alphaInputType=Bool
    rotationMode=BMM_Additive
    scaleMode=BMM_Additive
    scaleSpace=BCS_BoneSpace
    translationMode=BMM_Additive
    translationSpace=BCS_BoneSpace
    alphaBoolBlend {blendInTime=0.3, blendOption=ExpInOut, blendOutTime=0.3}
    boneToModify {boneName=dyn_shorts_hem_r}
    translation {y=-3}
   }
   n9 = "Transform (Modify) Bone - Bone: dyn_shorts_hem_l"(ComponentPose=@n8, bAlphaBoolEnabled=@n7, Translation="0,-3,0", Rotation="0,0,0", Scale="1,1,1") {
    alphaInputType=Bool
    rotationMode=BMM_Additive
    scaleMode=BMM_Additive
    scaleSpace=BCS_BoneSpace
    translationMode=BMM_Additive
    translationSpace=BCS_BoneSpace
    alphaBoolBlend {blendInTime=0.3, blendOption=ExpInOut, blendOutTime=0.3}
    boneToModify {boneName=dyn_shorts_hem_l}
    translation {y=-3}
   }
   n10 = Get ThighUpR() -> (ThighUpR:double)
   n11 = "Transform (Modify) Bone - Bone: dyn_hem_fr_r_MSTR"(ComponentPose=@n9, Alpha=@n10, Translation="0,4,28", Rotation="0,0,0", Scale="1,1,1") {
    rotationMode=BMM_Additive
    scaleMode=BMM_Additive
    translationMode=BMM_Additive
    alphaBoolBlend {blendInTime=0.3, blendOption=ExpInOut, blendOutTime=0.3}
    boneToModify {boneName=dyn_hem_fr_r_MSTR}
    translation {y=4, z=28}
   }
   n12 = Get ThighUpL() -> (ThighUpL:double)
   n13 = "Transform (Modify) Bone - Bone: dyn_hem_fr_l_MSTR"(ComponentPose=@n11, Alpha=@n12, Translation="0,4,29", Rotation="0,0,0", Scale="1,1,1") {
    rotationMode=BMM_Additive
    scaleMode=BMM_Additive
    translationMode=BMM_Additive
    alphaBoolBlend {blendInTime=0.3, blendOption=ExpInOut, blendOutTime=0.3}
    boneToModify {boneName=dyn_hem_fr_l_MSTR}
    translation {y=4, z=29}
   }
   n14 = "Transform (Modify) Bone - Bone: dyn_hem_r_MSTR"(ComponentPose=@n13, Alpha=@n10, Translation="0,0,3", Rotation="0,0,0", Scale="1,1,1") {
    rotationMode=BMM_Additive
    scaleMode=BMM_Additive
    translationMode=BMM_Additive
    alphaBoolBlend {blendInTime=0.3, blendOption=ExpInOut, blendOutTime=0.3}
    boneToModify {boneName=dyn_hem_r_MSTR}
    translation {z=3}
   }
   n15 = "Transform (Modify) Bone - Bone: dyn_hem_l_MSTR"(ComponentPose=@n14, Alpha=@n12, Translation="0,0,3", Rotation="0,0,0", Scale="1,1,1") {
    rotationMode=BMM_Additive
    scaleMode=BMM_Additive
    translationMode=BMM_Additive
    alphaBoolBlend {blendInTime=0.3, blendOption=ExpInOut, blendOutTime=0.3}
    boneToModify {boneName=dyn_hem_l_MSTR}
    translation {z=3}
   }
   n16 = "Transform (Modify) Bone - Bone: dyn_hem_fr_MSTR"(ComponentPose=@n15, Alpha=Get ThighUpDelta() -> (ThighUpDelta:double), Translation="0,0,7", Rotation="0,0,0", Scale="1,1,1") {
    rotationMode=BMM_Additive
    scaleMode=BMM_Additive
    translationMode=BMM_Additive
    alphaBoolBlend {blendInTime=0.3, blendOption=ExpInOut, blendOutTime=0.3}
    boneToModify {boneName=dyn_hem_fr_MSTR}
    translation {z=7}
   }
   n17 = Get CrouchIdleLeftHemCorrectionAlpha() -> (CrouchIdleLeftHemCorrectionAlpha:double)
   n18 = "Transform (Modify) Bone - Bone: dyn_hem_l_MSTR"(ComponentPose=@n16, Alpha=@n17, Translation="0,0,2.5", Rotation="0,0,0", Scale="1,1,1") {
    rotationMode=BMM_Additive
    scaleMode=BMM_Additive
    translationMode=BMM_Additive
    alphaBoolBlend {blendInTime=0.3, blendOption=ExpInOut, blendOutTime=0.3}
    alphaScaleBiasClamp {bInterpResult=true}
    boneToModify {boneName=dyn_hem_l_MSTR}
    translation {z=2.5}
   }
   n19 = "Transform (Modify) Bone - Bone: dyn_hem_fr_l_MSTR"(ComponentPose=@n18, Alpha=@n17, Translation="0,0,5", Rotation="0,0,0", Scale="1,1,1") {
    rotationMode=BMM_Additive
    scaleMode=BMM_Additive
    translationMode=BMM_Additive
    alphaBoolBlend {blendInTime=0.3, blendOption=ExpInOut, blendOutTime=0.3}
    alphaScaleBiasClamp {bInterpResult=true}
    boneToModify {boneName=dyn_hem_fr_l_MSTR}
    translation {z=5}
   }
   n20 = "Transform (Modify) Bone - Bone: attach_backpack"(ComponentPose=@n19, Alpha=1, Translation="0,2.5,0", Rotation="0,0,0", Scale="1,1,1") {
    translationMode=BMM_Additive
    translationSpace=BCS_BoneSpace
    alphaBoolBlend {blendOption=Linear}
    boneToModify {boneName=attach_backpack}
    translation {y=2.5}
   }
   n21 = "Transform (Modify) Bone - Bone: attach_cape"(ComponentPose=@n20, Alpha=1, Translation="0,3.5,0", Rotation="0,0,0", Scale="1,1,1") {
    translationMode=BMM_Additive
    translationSpace=BCS_BoneSpace
    alphaBoolBlend {blendOption=Linear}
    boneToModify {boneName=attach_cape}
    translation {y=3.5}
   }
   n22 = "Constraint - dyn_zipper_MSTR (PP)"!(ComponentPose=@n21, Alpha=1, ConstraintWeights_0=0.9, ConstraintWeights_1:float=0.1) -> (Pose) {
    alphaBoolBlend {blendOption=Linear}
    boneToModify {boneName=dyn_zipper_MSTR}
    constraintSetup [{targetBone {boneName=spine_05}}, {targetBone {boneName=neck_01}}]
    constraintWeights [1,1]
   }
   n23 = "Transform (Modify) Bone - Bone: dyn_back_neck"(ComponentPose=@n22, Alpha=1, Translation="0.000,-1.3,-0.000", Rotation="0,0,0", Scale="1,1,1") {
    translationMode=BMM_Additive
    translationSpace=BCS_BoneSpace
    alphaBoolBlend {blendOption=Linear}
    boneToModify {boneName=dyn_back_neck}
    translation {y=-1.3}
   }
   n24 = "Transform (Modify) Bone - Bone: dyn_back_neck"!(ComponentPose=@n23, bAlphaBoolEnabled:bool=Get bIsPlayingMeleeAnim(), Translation="0.000,-1.7,-0.000", Rotation="0,0,0", Scale="1,1,1") -> (Pose) {
    alphaInputType=Bool
    translationMode=BMM_Additive
    translationSpace=BCS_BoneSpace
    alphaBoolBlend {blendInTime=0.1, blendOption=Linear, blendOutTime=0.1}
    boneToModify {boneName=dyn_back_neck}
    translation {y=-1.7}
   }
   Output Pose(Result=Component To Local(ComponentPose=@n24)) {layerGroup=None}
  ]}
 EventGraph {nodes [
   n1 = Event BlueprintUpdateAnimation()
   > Set Delta Time X(Delta Time X=@n1.DeltaTimeX)
   > Is Valid(InputObject=Get Skeletal Mesh Component() -> (Skeletal Mesh Component:Skeletal Mesh Component Object Reference))
   > n2 = Sequence!(execute) -> (then_0, then_1, then_2:Exec)
   n3 = Get bIsMoving2D()
   > n4 = Set IsCrouchingNotMoving(IsCrouchingNotMoving:bool=AND Boolean(A=Get bIsCrouching(), B=NOT Boolean(A=@n3))) -> (Output_Get:bool)
   > Set CrouchIdleLeftHemCorrectionAlpha(CrouchIdleLeftHemCorrectionAlpha:double=SelectFloat(A=1, B=0, bPickA=@n4)) -> (Output_Get:double)
   Set EmoteCompensationAlpha(execute=@n2.then_1, EmoteCompensationAlpha:bool=NOT Boolean(A=Get bIsPlayingEmote())) -> (Output_Get:bool)
   n5 = Get Skeletal Mesh Component()
   n6 = BreakTransform(InTransform=GetDeltaTransformFromRefPose(self=@n5, BoneName=thigh_l, BaseName=pelvis))
   n7 = Set ThighUpL(execute=@n2.then_2, ThighUpL:double=MapRangeClamped(Value=BreakRotator(InRot=@n6.Rotation).Yaw, InRangeA=-20, InRangeB=-120)) -> (Output_Get:double)
   n8 = BreakRotator(InRot=BreakTransform(InTransform=GetDeltaTransformFromRefPose(self=@n5, BoneName=thigh_r, BaseName=pelvis)).Rotation)
   n9 = TryGetPawnOwner()
   n10 = Get PawnVelocity()
   n11 = float * float(
    A=Dot Product!(
     A:Vector=Normalize!(A:Vector=@n10, Tolerance:float=0.0001) -> (ReturnValue:Vector) {is=title K2Node_CallFunction}
     B:Vector=GetActorForwardVector!(self:Actor Object Reference=@n9) -> (ReturnValue:Vector)
    ) -> (ReturnValue:double)
    B="To Float (Boolean)"(InBool=Get bIsCrouching())
   ) {is=title K2Node_PromotableOperator}
   > n12 = Set ThighUpR(ThighUpR:double=float * float(
     A=MapRangeClamped(Value=@n8.Yaw, InRangeA=-20, InRangeB=-120)
     B=MapRangeClamped(Value=@n11, InRangeA=0, InRangeB=1, OutRangeA=1, OutRangeB=0.9)
    )) -> (Output_Get:double)
   > Set ThighUpDelta(ThighUpDelta:double=float / float(A=float + float(A=@n7, B=@n12), B=2)) -> (Output_Get:double)
   > Set IsMoving?(IsMoving?:bool=OR Boolean(A=Get bIsJumping(), B=Get bIsSprinting())) -> (Output_Get:bool)
   Event OnCharacterCustomizationLoaded()
   > Set Skeletal Mesh Component(Skeletal Mesh Component:Skeletal Mesh Component Object Reference=GetOwnerSkeletalMeshComponent()) -> (Output_Get:Skeletal Mesh Component Object Reference)
  ]}
]
variables [
 CrouchIdleLeftHemCorrectionAlpha {type=number}
 Delta Time X {type=number}
 EmoteCompensationAlpha {type=boolean}
 IsCrouchingNotMoving {type=boolean}
 IsMoving? {type=boolean}
 Skeletal Mesh Component {type=SkeletalMeshComponent}
 ThighUpDelta {type=number}
 ThighUpL {type=number}
 ThighUpLCrouching {type=number}
 ThighUpR {type=number}
 ThighUpRCrouching {type=number}
]
